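# Sound cue for the out-of-ammo sound shared by bow weapons.
# A single wave player feeds one bow out-of-ammo wave through a Modulator at the root, left at default settings.
# It plays as weapon foley, using that sound class, attenuation and concurrency, with reduced volume and an overridden falloff distance.
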SoundCue /PrimalGameplay/Sounds/Weapons/Bows/Shared/SC_Bow_Shared_OutOfAmmo
root=Modulator_4
nodes [
 {id=Modulator_4, children [WavePlayer_0]}
 {id=WavePlayer_0, properties {soundWaveAssetPtr=&/Bow_Shared_OutOfAmmo_01}}
]
properties {
 attenuationSettings=&AS_Fort_WeaponFoley
 maxDistance=2250
 soundClassObject=&Fort_WeaponFoley
 volumeMultiplier=0.33
 attenuationOverrides {falloffDistance=1600}
 concurrencySet [&CS_Fort_WeaponFoley]
}
